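# Sound cue for coaster spark spot ambience at the Hollow Base takeover point of interest, classed as outdoor ambient.
# A Random node picks one of eight equally weighted spark waves and feeds a Modulator root that varies pitch.
# It plays at reduced volume through a distant one-shot creak attenuation with falloff and stereo spread overrides, with bus and submix sends disabled.
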SoundCue /Game/Sounds/Artemis/POI/Hollow_Base_Takover/Spots/Spot_HollowBase_Coaster_Spark_Cue
root=Modulator_0
nodes [
 {id=Modulator_0, children [Random_1], properties {pitchMin=0.9}}
 {
  id=Random_1
  children [WavePlayer_8, WavePlayer_9, WavePlayer_10, WavePlayer_11, WavePlayer_12, WavePlayer_13, WavePlayer_14, WavePlayer_15]
  properties {weights [1,1,1,1,1,1,1,1]}
 }
 {id=WavePlayer_10, properties {soundWaveAssetPtr=&/Spot_HollowBase_Coaster_Spark_03}}
 {id=WavePlayer_11, properties {soundWaveAssetPtr=&/Spot_HollowBase_Coaster_Spark_04}}
 {id=WavePlayer_12, properties {soundWaveAssetPtr=&/Spot_HollowBase_Coaster_Spark_05}}
 {id=WavePlayer_13, properties {soundWaveAssetPtr=&/Spot_HollowBase_Coaster_Spark_06}}
 {id=WavePlayer_14, properties {soundWaveAssetPtr=&/Spot_HollowBase_Coaster_Spark_07}}
 {id=WavePlayer_15, properties {soundWaveAssetPtr=&/Spot_HollowBase_Coaster_Spark_08}}
 {id=WavePlayer_8, properties {soundWaveAssetPtr=&/Spot_HollowBase_Coaster_Spark_01}}
 {id=WavePlayer_9, properties {soundWaveAssetPtr=&/Spot_HollowBase_Coaster_Spark_02}}
]
properties {
 attenuationSettings=/Game/Sounds/Attenuation/Fort_TileAmbient_OneShot_Distant_Creak_Att
 bEnableBusSends=false
 bEnableSubmixSends=false
 maxDistance=8200
 soundClassObject=&Fort_Ambient_Outdoors_Athena
 volumeMultiplier=0.301
 attenuationOverrides {falloffDistance=6000, stereoSpread=10000, attenuationShapeExtents {x=2000}}
}
