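# Sound cue for spot ambience of metal coaster track movement at the Hollow Base point of interest, in the outdoor ambient sound class.
# A Random node picks among five wave variations and one empty slot at equal weight, feeding a Modulator with unchanged settings.
# It uses a shared, less-positional one-shot ambient attenuation asset.
# A concurrency override allows one instance per owner and blocks new plays while it sounds.
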SoundCue /Game/Sounds/Artemis/POI/Hollow_Base_Takover/Spots/Spot_HollowBase_Coaster_Track_Metal_Movement_Cue
root=Modulator_0
nodes [
 {id=Modulator_0, children [Random_0]}
 {id=Random_0, children [WavePlayer_0, WavePlayer_1, WavePlayer_2, WavePlayer_3, null, WavePlayer_11], properties {weights [1,1,1,1,1,1]}}
 {
  id=WavePlayer_0
  properties {soundWaveAssetPtr=&/Spot_HollowBase_Coaster_Track_Metal_Movement_01}
 }
 {
  id=WavePlayer_1
  properties {soundWaveAssetPtr=&/Spot_HollowBase_Coaster_Track_Metal_Movement_02}
 }
 {
  id=WavePlayer_11
  properties {soundWaveAssetPtr=&/Spot_HollowBase_Coaster_Track_Metal_Movement_06}
 }
 {
  id=WavePlayer_2
  properties {soundWaveAssetPtr=&/Spot_HollowBase_Coaster_Track_Metal_Movement_03}
 }
 {
  id=WavePlayer_3
  properties {soundWaveAssetPtr=&/Spot_HollowBase_Coaster_Track_Metal_Movement_04}
 }
]
properties {
 attenuationSettings=/Game/Sounds/Attenuation/Fort_TileAmbient_OneShot_LessPositional_Log_Att
 bEnableBusSends=false
 bEnableSubmixSends=false
 bOverrideConcurrency=true
 maxDistance=6200
 soundClassObject=&Fort_Ambient_Outdoors_Athena
 volumeMultiplier=2.506
 attenuationOverrides {bAttenuate=false}
 concurrencyOverrides {bLimitToOwner=true, maxCount=1, resolutionRule=PreventNew}
}
