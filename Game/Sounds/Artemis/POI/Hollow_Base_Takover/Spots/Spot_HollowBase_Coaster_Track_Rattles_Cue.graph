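# Sound cue for spot ambient coaster track rattles at the Hollow Base POI, routed to the outdoor ambient sound class.
# A Random node picks one of six equally weighted rattle waves and feeds a root Modulator node.
# It uses a shared less-positional one-shot attenuation asset, with boosted volume and bus and submix sends disabled.
# A concurrency override allows one instance per owner and blocks new plays while one is active.
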SoundCue /Game/Sounds/Artemis/POI/Hollow_Base_Takover/Spots/Spot_HollowBase_Coaster_Track_Rattles_Cue
root=Modulator_0
nodes [
 {id=Modulator_0, children [Random_1]}
 {
  id=Random_1
  children [WavePlayer_5, WavePlayer_6, WavePlayer_7, WavePlayer_8, WavePlayer_9, WavePlayer_10]
  properties {weights [1,1,1,1,1,1]}
 }
 {id=WavePlayer_10, properties {soundWaveAssetPtr=&/Spot_HollowBase_Coaster_Track_Rattle_06}}
 {id=WavePlayer_5, properties {soundWaveAssetPtr=&/Spot_HollowBase_Coaster_Track_Rattle_01}}
 {id=WavePlayer_6, properties {soundWaveAssetPtr=&/Spot_HollowBase_Coaster_Track_Rattle_02}}
 {id=WavePlayer_7, properties {soundWaveAssetPtr=&/Spot_HollowBase_Coaster_Track_Rattle_03}}
 {id=WavePlayer_8, properties {soundWaveAssetPtr=&/Spot_HollowBase_Coaster_Track_Rattle_04}}
 {id=WavePlayer_9, properties {soundWaveAssetPtr=&/Spot_HollowBase_Coaster_Track_Rattle_05}}
]
properties {
 attenuationSettings=/Game/Sounds/Attenuation/Fort_TileAmbient_OneShot_LessPositional_Log_Att
 bEnableBusSends=false
 bEnableSubmixSends=false
 bOverrideConcurrency=true
 maxDistance=6200
 soundClassObject=&Fort_Ambient_Outdoors_Athena
 volumeMultiplier=2.506
 attenuationOverrides {bAttenuate=false}
 concurrencyOverrides {bLimitToOwner=true, maxCount=1, resolutionRule=PreventNew}
}
