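# Sound cue that plays distant roller coaster chain pull spot sounds as ambient one-shots for the Hollow Base takeover POI.
# A Random node picks from nine equally weighted wave variations and feeds a Modulator for slight pitch and volume variation.
# It uses the ambient one-shots sound class and a shared distant creak attenuation asset, at reduced volume with bus and submix sends disabled.
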SoundCue /Game/Sounds/Artemis/POI/Hollow_Base_Takover/Spots/Spot_Roller_Coaster_Chain_Pulls_Distant_Cue
root=Modulator_0
nodes [
 {id=Modulator_0, children [Random_0], properties {pitchMax=1, pitchMin=0.9, volumeMax=1.1}}
 {
  id=Random_0
  children [WavePlayer_0,WavePlayer_1,WavePlayer_2,WavePlayer_3,WavePlayer_4,WavePlayer_5,WavePlayer_6,WavePlayer_7,WavePlayer_8]
  properties {weights [1,1,1,1,1,1,1,1,1]}
 }
 {
  id=WavePlayer_0
  properties {soundWaveAssetPtr=&/Spot_Roller_Coaster_Chain_Pulls_Distant_09}
 }
 {
  id=WavePlayer_1
  properties {soundWaveAssetPtr=&/Spot_Roller_Coaster_Chain_Pulls_Distant_01}
 }
 {
  id=WavePlayer_2
  properties {soundWaveAssetPtr=&/Spot_Roller_Coaster_Chain_Pulls_Distant_02}
 }
 {
  id=WavePlayer_3
  properties {soundWaveAssetPtr=&/Spot_Roller_Coaster_Chain_Pulls_Distant_03}
 }
 {
  id=WavePlayer_4
  properties {soundWaveAssetPtr=&/Spot_Roller_Coaster_Chain_Pulls_Distant_04}
 }
 {
  id=WavePlayer_5
  properties {soundWaveAssetPtr=&/Spot_Roller_Coaster_Chain_Pulls_Distant_05}
 }
 {
  id=WavePlayer_6
  properties {soundWaveAssetPtr=&/Spot_Roller_Coaster_Chain_Pulls_Distant_06}
 }
 {
  id=WavePlayer_7
  properties {soundWaveAssetPtr=&/Spot_Roller_Coaster_Chain_Pulls_Distant_07}
 }
 {
  id=WavePlayer_8
  properties {soundWaveAssetPtr=&/Spot_Roller_Coaster_Chain_Pulls_Distant_08}
 }
]
properties {
 attenuationSettings=/Game/Sounds/Attenuation/Fort_TileAmbient_OneShot_Distant_Creak_Att
 bEnableBusSends=false
 bEnableSubmixSends=false
 maxDistance=8200
 soundClassObject=&Fort_Ambient_OneShots
 volumeMultiplier=0.501
 attenuationOverrides {falloffDistance=6000, stereoSpread=10000, attenuationShapeExtents {x=2000}}
}
