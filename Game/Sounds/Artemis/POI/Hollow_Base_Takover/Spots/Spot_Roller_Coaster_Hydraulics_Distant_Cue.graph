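# Sound cue for a distant roller coaster hydraulics spot sound at the Hollow Base Takover point of interest.
# A random node picks one of three equally weighted wave variations, and a modulator above it adds slight upward pitch variation.
# It plays in the ambient one-shot sound class at reduced volume, using a wind one-shot attenuation asset, with bus and submix sends disabled.
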SoundCue /Game/Sounds/Artemis/POI/Hollow_Base_Takover/Spots/Spot_Roller_Coaster_Hydraulics_Distant_Cue
root=Modulator_0
nodes [
 {id=Modulator_0, children [Random_4], properties {pitchMax=1.1, pitchMin=1}}
 {id=Random_4, children [WavePlayer_38, WavePlayer_39, WavePlayer_40], properties {weights [1,1,1]}}
 {
  id=WavePlayer_38
  properties {soundWaveAssetPtr=&/Spot_Roller_Coaster_Hydraulics_Distant_02}
 }
 {
  id=WavePlayer_39
  properties {soundWaveAssetPtr=&/Spot_Roller_Coaster_Hydraulics_Distant_03}
 }
 {
  id=WavePlayer_40
  properties {soundWaveAssetPtr=&/Spot_Roller_Coaster_Hydraulics_Distant_04}
 }
]
properties {
 attenuationSettings=/Game/Sounds/Attenuation/Fort_TileAmbient_OneShot_Wind_Att
 bEnableBusSends=false
 bEnableSubmixSends=false
 maxDistance=8000
 soundClassObject=&Fort_Ambient_OneShots
 volumeMultiplier=0.501
 attenuationOverrides {falloffDistance=6000, stereoSpread=10000, attenuationShapeExtents {x=2000}}
}
