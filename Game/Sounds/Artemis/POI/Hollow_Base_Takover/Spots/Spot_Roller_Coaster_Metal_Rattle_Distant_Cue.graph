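# Sound cue for distant roller coaster metal rattle spot one-shots in the Hollow Base Takover POI ambience.
# A Random node picks among nine equally weighted rattle wave variations, then a Modulator applies a slight random upward pitch shift.
# It plays at reduced volume in the ambient one-shots sound class, using a distant creak attenuation with overridden falloff and stereo spread, and bus and submix sends disabled.
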SoundCue /Game/Sounds/Artemis/POI/Hollow_Base_Takover/Spots/Spot_Roller_Coaster_Metal_Rattle_Distant_Cue
root=Modulator_0
nodes [
 {id=Modulator_0, children [Random_1], properties {pitchMax=1.1, pitchMin=1}}
 {
  id=Random_1
  children [WavePlayer_14, WavePlayer_15, WavePlayer_16, WavePlayer_17, WavePlayer_18, WavePlayer_19, WavePlayer_20, WavePlayer_21, WavePlayer_22]
  properties {weights [1,1,1,1,1,1,1,1,1]}
 }
 {
  id=WavePlayer_14
  properties {soundWaveAssetPtr=&/Spot_Roller_Coaster_Metal_Rattles_Distant_01}
 }
 {
  id=WavePlayer_15
  properties {soundWaveAssetPtr=&/Spot_Roller_Coaster_Metal_Rattles_Distant_02}
 }
 {
  id=WavePlayer_16
  properties {soundWaveAssetPtr=&/Spot_Roller_Coaster_Metal_Rattles_Distant_03}
 }
 {
  id=WavePlayer_17
  properties {soundWaveAssetPtr=&/Spot_Roller_Coaster_Metal_Rattles_Distant_04}
 }
 {
  id=WavePlayer_18
  properties {soundWaveAssetPtr=&/Spot_Roller_Coaster_Metal_Rattles_Distant_05}
 }
 {
  id=WavePlayer_19
  properties {soundWaveAssetPtr=&/Spot_Roller_Coaster_Metal_Rattles_Distant_06}
 }
 {
  id=WavePlayer_20
  properties {soundWaveAssetPtr=&/Spot_Roller_Coaster_Metal_Rattles_Distant_07}
 }
 {
  id=WavePlayer_21
  properties {soundWaveAssetPtr=&/Spot_Roller_Coaster_Metal_Rattles_Distant_08}
 }
 {
  id=WavePlayer_22
  properties {soundWaveAssetPtr=&/Spot_Roller_Coaster_Metal_Rattles_Distant_09}
 }
]
properties {
 attenuationSettings=/Game/Sounds/Attenuation/Fort_TileAmbient_OneShot_Distant_Creak_Att
 bEnableBusSends=false
 bEnableSubmixSends=false
 maxDistance=8200
 soundClassObject=&Fort_Ambient_OneShots
 volumeMultiplier=0.501
 attenuationOverrides {falloffDistance=6000, stereoSpread=10000, attenuationShapeExtents {x=2000}}
}
